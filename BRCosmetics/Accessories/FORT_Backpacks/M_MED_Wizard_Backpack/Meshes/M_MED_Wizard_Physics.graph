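# Physics asset for the medium male Wizard backpack cosmetic, giving its dangling gourds, feathers and loops bodies for secondary motion.
# Two limited-swing gourds, a loopA chain ending in three feather segments, and a six-link loopB chain hang from the backpack_root capsule.
# Small spheres and capsules form the chains, with some feather and root joints left free to swing.
# A dyn_pelvisCOL body adds three named boxes as back, feather-loop and necklace-loop colliders.
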PhysicsAsset /BRCosmetics/Accessories/FORT_Backpacks/M_MED_Wizard_Backpack/Meshes/M_MED_Wizard_Physics
bodies {
 backpack_root [{shapeType=Capsule, length=23.0423, radius=6.4306, center=<-8.7225 -4.0067 0>, rotation=<rot 0.1885 -89.9957 2.3465>}]
 dyn_feather_01 [{shapeType=Sphere, radius=2.804, center=<4.4258 -0.0146 -0.0035>}]
 dyn_feather_02 [{shapeType=Sphere, radius=2.4581, center=<1.1909 0.0001 0>}]
 dyn_feather_03 [{shapeType=Sphere, radius=2.5617, center=<1.8852 0 0>}]
 dyn_gourdA [{shapeType=Sphere, radius=3.1827, center=<4.1061 2.5749 -2.0775>}]
 dyn_gourdB [{shapeType=Sphere, radius=2.7687, center=<4.7582 0.0457 0.327>}]
 dyn_loopA_01 [{shapeType=Sphere, radius=1.4889, center=<-0.0019 -0.0001 0.0124>}]
 dyn_loopA_02 [{shapeType=Capsule, length=0.8378, radius=0.8555, center=<0.9501 0.0005 -0.0001>, rotation=<rot 0 89.9192 -89.16>}]
 dyn_loopA_03 [{shapeType=Sphere, radius=1.0086, center=<0.6934 0 0.005>}]
 dyn_loopB_01 [{shapeType=Sphere, radius=0.505}]
 dyn_loopB_02 [{shapeType=Sphere, radius=1.533, center=<1.2457 0 0>}]
 dyn_loopB_03 [{shapeType=Sphere, radius=1.6664, center=<1.1951 -0.0001 0>}]
 dyn_loopB_04 [{shapeType=Sphere, radius=1.6433, center=<1.3344 -0.0001 0>}]
 dyn_loopB_05 [{shapeType=Sphere, radius=1.5218}]
 dyn_loopB_06 [{shapeType=Sphere, radius=1.7824, center=<0 0 -2.2226>}]
 dyn_pelvisCOL [
  {
   shapeType=Box
   extentX=27.4134
   extentY=1.01
   extentZ=45.4044
   shapeName=LargeBackCOL
   center=<12.7782 0 11.0092>
   rotation=<rot 0 -89.9998 8.4196>
  }
  {
   shapeType=Box
   extentX=2.0053
   extentY=13.5535
   extentZ=11.2947
   shapeName=FeatherLoopCOL
   center=<16.4868 6.7024 17.6717>
   rotation=<rot -9.4098 0 0>
  }
  {
   shapeType=Box
   extentX=4.4
   extentY=10.3209
   extentZ=7.5
   shapeName=NecklaceLoopCOL
   center=<13.1211 -8.1407 5.6351>
   rotation=<rot -8.7025 0 0>
  }
 ]
}
constraints [
 {bone1Name=dyn_loopA_01, bone2Name=backpack_root, swing1LimitDegrees=45, swing2LimitDegrees=1, twistLimitDegrees=1}
 {bone1Name=dyn_loopA_03, bone2Name=dyn_loopA_02, swing1LimitDegrees=30, swing2LimitDegrees=30, twistLimitDegrees=30}
 {
  bone1Name=dyn_feather_01
  bone2Name=dyn_loopA_03
  swing1LimitDegrees=35
  swing1Motion=Free
  swing2LimitDegrees=35
  swing2Motion=Free
  twistLimitDegrees=1
 }
 {bone1Name=dyn_feather_02, bone2Name=dyn_feather_01, swing1LimitDegrees=35, swing1Motion=Free, swing2LimitDegrees=5, twistLimitDegrees=1}
 {bone1Name=dyn_feather_03, bone2Name=dyn_feather_02, swing1LimitDegrees=35, swing1Motion=Free, swing2LimitDegrees=10, twistLimitDegrees=1}
 {bone1Name=dyn_loopA_02, bone2Name=dyn_loopA_01, swing1LimitDegrees=30, swing2LimitDegrees=30, twistLimitDegrees=30}
 {
  bone1Name=dyn_pelvisCOL
  bone2Name=backpack_root
  swing1LimitDegrees=45
  swing1Motion=Free
  swing2LimitDegrees=45
  swing2Motion=Free
  twistLimitDegrees=45
  twistMotion=Free
 }
 {bone1Name=dyn_gourdA, bone2Name=backpack_root, swing1LimitDegrees=10, swing2LimitDegrees=10, twistLimitDegrees=2}
 {bone1Name=dyn_gourdB, bone2Name=backpack_root, swing1LimitDegrees=20, swing2LimitDegrees=20, twistLimitDegrees=2}
 {
  bone1Name=dyn_loopB_01
  bone2Name=backpack_root
  swing1LimitDegrees=45
  swing1Motion=Free
  swing2LimitDegrees=45
  swing2Motion=Free
  twistLimitDegrees=45
  twistMotion=Free
 }
 {bone1Name=dyn_loopB_02, bone2Name=dyn_loopB_01, swing1LimitDegrees=20, swing2LimitDegrees=20, twistLimitDegrees=5}
 {bone1Name=dyn_loopB_03, bone2Name=dyn_loopB_02, swing1LimitDegrees=20, swing2LimitDegrees=20, twistLimitDegrees=5}
 {bone1Name=dyn_loopB_04, bone2Name=dyn_loopB_03, swing1LimitDegrees=20, swing2LimitDegrees=20, twistLimitDegrees=5}
 {bone1Name=dyn_loopB_05, bone2Name=dyn_loopB_04, swing1LimitDegrees=20, swing2LimitDegrees=20, twistLimitDegrees=5}
 {bone1Name=dyn_loopB_06, bone2Name=dyn_loopB_05, swing1LimitDegrees=20, swing2LimitDegrees=5, twistLimitDegrees=20}
]
