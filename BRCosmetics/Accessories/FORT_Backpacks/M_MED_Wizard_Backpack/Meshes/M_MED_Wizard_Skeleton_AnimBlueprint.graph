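# Post-pose AnimBlueprint for the Wizard backpack cosmetic, adding secondary motion to the input pose.
# AnimDynamics swings the wand and four pads, then RigidBody simulates the physics asset before constraints and look-at tie loop bones to the feather.
# The event graph derives backpack root rotation from spine bend and a crouch alpha that offsets the pelvis collision base and gourd.
# Skydiving feeds height-based Perlin noise into RigidBody external force, dampened under an open parachute.
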AnimBlueprint /BRCosmetics/Accessories/FORT_Backpacks/M_MED_Wizard_Backpack/Meshes/M_MED_Wizard_Skeleton_AnimBlueprint
target_skeleton=&/M_MED_Wizard_Skeleton
graphs [
 AnimGraph {nodes [
   AnimGraphNode_SubInput_0 = Input Pose()
   n1 = Anim Dynamics - Bone: dyn_wand(ComponentPose=Local To Component(LocalPose=@AnimGraphNode_SubInput_0), Alpha=1, ExternalForce="0,0,0") {
    angularSpringConstant=100
    bAngularSpring=true
    bGravityOverrideInSimSpace=true
    bUseGravityOverride=true
    simulationSpace=RootRelative
    alphaBoolBlend {blendOption=Linear}
    boundBone {boneName=dyn_wand}
    componentAppliedLinearAccClamp=<1000 1000 1000>
    componentLinearAccScale=<0.4 0.4 0.4>
    componentLinearVelScale=<0.35 0.35 0.35>
    gravityOverride {z=-980}
    physicsBodyDefinitions [{
      sphereCollisionRadius=0
      boundBone {boneName=dyn_wand}
      boxExtents=<8 8 8>
      constraintSetup {twistAxis=AxisY, angularLimitsMax=<4 5 4>, angularLimitsMin=<-3 -5 -4>, angularTarget=<0 0 0>}
      localJointOffset=<-5 0 3>
     }]
    relativeSpaceBone {boneName=backpack_root}
   }
   n2 = Anim Dynamics - Bone: dyn_padA_l(ComponentPose=@n1, Alpha=1, ExternalForce="0,0,0") {
    angularDampingOverride=0.9999
    angularSpringConstant=1500
    bAngularSpring=true
    bGravityOverrideInSimSpace=true
    bOverrideAngularDamping=true
    bUseGravityOverride=true
    alphaBoolBlend {blendOption=Linear}
    boundBone {boneName=dyn_padA_l}
    componentAppliedLinearAccClamp=<1000 1000 1000>
    componentLinearAccScale=<0.6 0.6 0.6>
    componentLinearVelScale=<0.35 0.35 0.35>
    gravityOverride {z=-980}
    physicsBodyDefinitions [{
      sphereCollisionRadius=0
      boundBone {boneName=dyn_padA_l}
      boxExtents=<10 10 10>
      constraintSetup {twistAxis=AxisZ, angularLimitsMax=<1 2 3>, angularLimitsMin=<-1 -6 -3>, angularTarget=<1 0 0>}
      localJointOffset=<0 0 -1>
     }]
    relativeSpaceBone {boneName=dyn_padA_l}
   }
   n3 = Anim Dynamics - Bone: dyn_padB_l(ComponentPose=@n2, Alpha=1, ExternalForce="0,0,0") {
    angularDampingOverride=0.9999
    angularSpringConstant=1500
    bAngularSpring=true
    bGravityOverrideInSimSpace=true
    bOverrideAngularDamping=true
    bUseGravityOverride=true
    alphaBoolBlend {blendOption=Linear}
    boundBone {boneName=dyn_padB_l}
    componentAppliedLinearAccClamp=<1000 1000 1000>
    componentLinearAccScale=<0.7 0.7 0.7>
    componentLinearVelScale=<0.4 0.4 0.4>
    gravityOverride {z=-980}
    physicsBodyDefinitions [{
      sphereCollisionRadius=0
      boundBone {boneName=dyn_padB_l}
      boxExtents=<15 8 8>
      constraintSetup {twistAxis=AxisZ, angularLimitsMax=<2 3 3>, angularLimitsMin=<-2 -6 -3>, angularTarget=<0 0 0>}
      localJointOffset=<0 0 -1>
     }]
    relativeSpaceBone {boneName=dyn_padB_l}
   }
   n4 = Anim Dynamics - Bone: dyn_padA_r(ComponentPose=@n3, Alpha=0.65, ExternalForce="0,0,0") {
    angularDampingOverride=0.9999
    angularSpringConstant=1000
    bAngularSpring=true
    bGravityOverrideInSimSpace=true
    bOverrideAngularDamping=true
    bUseGravityOverride=true
    alphaBoolBlend {blendOption=Linear}
    boundBone {boneName=dyn_padA_r}
    componentAppliedLinearAccClamp=<1000 1000 1000>
    componentLinearAccScale=<0.5 0.5 0.5>
    componentLinearVelScale=<0.35 0.35 0.35>
    gravityOverride {z=-980}
    physicsBodyDefinitions [{
      sphereCollisionRadius=0
      boundBone {boneName=dyn_padA_r}
      boxExtents=<10 10 10>
      constraintSetup {angularLimitsMax=<2 3 5>, angularLimitsMin=<-2 -4 -5>, angularTarget=<1 0 0>}
      localJointOffset=<1 0 1>
     }]
    relativeSpaceBone {boneName=dyn_padA_r}
   }
   n5 = Anim Dynamics - Bone: dyn_padB_r(ComponentPose=@n4, Alpha=1, ExternalForce="0,0,0") {
    angularDampingOverride=0.9999
    angularSpringConstant=1500
    bAngularSpring=true
    bGravityOverrideInSimSpace=true
    bOverrideAngularDamping=true
    bUseGravityOverride=true
    alphaBoolBlend {blendOption=Linear}
    boundBone {boneName=dyn_padB_r}
    componentAppliedLinearAccClamp=<1000 1000 1000>
    componentLinearAccScale=<0.4 0.4 0.4>
    componentLinearVelScale=<0.4 0.4 0.4>
    gravityOverride {z=-980}
    physicsBodyDefinitions [{
      sphereCollisionRadius=0
      boundBone {boneName=dyn_padB_r}
      boxExtents=<11 10 5>
      constraintSetup {twistAxis=AxisZ, angularLimitsMax=<2 3 2>, angularLimitsMin=<-2 -3 -2>, angularTarget=<0 0 0>}
      localJointOffset=<1 0 1>
     }]
    relativeSpaceBone {boneName=dyn_padB_r}
   }
   n6 = RigidBody(ComponentPose=@n5, ExternalForce=Get falling_anim_dyn_up_force_vec3() -> (falling_anim_dyn_up_force_vec3:Vector)) {
    bOverrideWorldGravity=true
    overridePhysicsAsset=&/M_MED_Wizard_Physics
    simulationSpace=BaseBoneSpace
    alphaBoolBlend {blendOption=Linear}
    baseBoneRef {boneName=backpack_root}
    componentAppliedLinearAccClamp=<100 100 100>
    componentLinearAccScale=<0.5 0.5 0.5>
    componentLinearVelScale=<0.5 0.5 0.5>
    overrideWorldGravity {z=-980}
   }
   n7 = "Constraint - dyn_loopA_09 (T)"(ComponentPose=@n6, Alpha=1, ConstraintWeights_0=1) {
    alphaBoolBlend {blendOption=Linear}
    boneToModify {boneName=dyn_loopA_09}
    constraintSetup [{transformType=Translation, targetBone {boneName=dyn_feather_01}}]
    constraintWeights [1]
   }
   n8 = Look At - Bone: dyn_loopA_07(ComponentPose=@n7, Alpha=1) {
    interpolationTime=10
    alphaBoolBlend {blendOption=Linear}
    boneToModify {boneName=dyn_loopA_07}
    lookAtLocation {x=0}
    lookAtTarget {boneReference {boneName=dyn_feather_01}}
    lookAt_Axis {axis=<1 0>}
   }
   n9 = "Transform (Modify) Bone - Bone: backpack_root"(ComponentPose=@n8, Alpha=1, Translation="0,0,0", Rotation=Get backpack_rotate(), Scale="1,1,1") {rotationMode=BMM_Additive, rotationSpace=BCS_BoneSpace, alphaBoolBlend {blendOption=Linear}, boneToModify {boneName=backpack_root}}
   n10 = Get Crouching() -> (Crouching:double)
   n11 = "Transform (Modify) Bone - Bone: dyn_pelvisCOLBase"!(ComponentPose=@n9, Alpha=@n10, Rotation="0,0,0", Scale="1,1,1") -> (Pose) {
    rotationSpace=BCS_BoneSpace
    translationMode=BMM_Additive
    translationSpace=BCS_BoneSpace
    alphaBoolBlend {blendOption=Linear}
    alphaScaleBiasClamp {bInterpResult=true, interpSpeedDecreasing=3, interpSpeedIncreasing=3}
    boneToModify {boneName=dyn_pelvisCOLBase}
    translation {x=-10}
   }
   n12 = "Transform (Modify) Bone - Bone: dyn_gourdA"(ComponentPose=@n11, Alpha=@n10, Translation="0,0,0", Rotation="0,60,0", Scale="1,1,1") {
    rotationMode=BMM_Additive
    rotationSpace=BCS_BoneSpace
    translationMode=BMM_Additive
    alphaBoolBlend {blendOption=Linear}
    alphaScaleBiasClamp {bInterpResult=true}
    boneToModify {boneName=dyn_gourdA}
    rotation {yaw=60}
   }
   n13 = "Constraint - dyn_gourdA (R)"(ComponentPose=@n12, Alpha=0.5, ConstraintWeights_0=1) {
    alphaBoolBlend {blendOption=Linear}
    boneToModify {boneName=dyn_gourdA}
    constraintSetup [{transformType=Rotation, targetBone {boneName=dyn_wand}}]
    constraintWeights [1]
   }
   Output Pose(Result=Component To Local(ComponentPose=@n13)) {layerGroup=None}
  ]}
 EventGraph {nodes [
   Event BlueprintUpdateAnimation()
   n1 = GetOwnerSkeletalMeshComponent()
   n2 = BreakRotator(InRot=BreakTransform(InTransform=GetDeltaTransformFromRefPose(self=@n1, BoneName=spine_05, BaseName=spine_01)).Rotation)
   > Set backpack_rotate(backpack_rotate=MakeRotator(Yaw=MapRangeClamped(Value=@n2.Yaw, InRangeA=12, InRangeB=-30, OutRangeA=8, OutRangeB=25)))
   > Set Crouching(Crouching:double="To Float (Boolean)"(InBool=Get bIsCrouching())) -> (Output_Get:double)
   n3 = Get PawnLastLocation() -> (PawnLastLocation:Vector)
   n4 = SelectFloat(
    A=Make Perlin Noise and Remap!(Value:float=BreakVector(InVec=@n3).Z, RangeOutMin:float=-15000, RangeOutMax:float=15000) -> (ReturnValue:float)
    B=0
    bPickA=Get bIsSkyDiving()
   )
   n5 = SelectFloat(A=float * float(A=@n4, B=0.2), B=@n4, bPickA=Get bIsParachuteOpened())
   > Set falling_anim_dyn_up_force_vec3(falling_anim_dyn_up_force_vec3:Vector=MakeVector(X=@n5, Y=@n5)) -> (Output_Get:Vector)
  ]}
]
variables [
 backpack_rotate {type=Rotator}
 Crouching {type=number}
 falling_anim_dyn_up_force_vec3 {type=Vector}
]
